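# Niagara water splash effect for the soccer traversal emote, with its two enabled emitters gated by User.bIsInWater.
# A local DirectionalSpawning module emits droplet sprites, launched upward along the owner's velocity from the water plane, then pulled down by gravity.
# Foam meshes spawn at a steady rate, expand while constrained to that plane, and pass water body index and depth to their material.
# Bubble, dust, smoke, arm-splash, test and other drop emitters are disabled.
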NiagaraSystem /BRCosmetics/Effects/Fort_Effects/Emotes/SoccerTraversal/NS_SoccerTraversal_Emote_WaterSplash
dependencies {
 usedDataInterfaces [
  &NiagaraDataInterfaceCurve
  &NiagaraDataInterfaceVectorCurve
  &NiagaraDataInterfaceVector2DCurve
 ]
 usedDynamicInputs []
 usedModules [&SystemState]
}
emitters {
 Bubbles001 {
  bEnabled=false
  emitterUpdateScript [
   &EmitterState()
   &SpawnRate(
    Spawn Probability=&MakeCustomFloatFromBool(Bool=$User.bIsInWater, True Float=1, False Float=0)
    SpawnRate=16
    Use Spawn Probability=true
   ) {module=SpawnRate001}
  ]
  particleSpawnScript [
   /Niagara/Modules/Spawn/Initialization/InitializeParticle(
    Lifetime=&UniformRangedFloat(Minimum=0.4, Maximum=0.6)
    Mesh Scale=&VectorFromFloat(Value=100)
    Module.Write Color=true
    Module.Write Lifetime=true
    Module.Write Mass=true
    Module.Write Material Random=true
    Module.Write Position=false
    Module.Write Scale=true
    Module.Write SpriteRotation=false
    Module.Write SpriteSize=false
    Position=&SimulationPosition()
   )
   &InitialMeshOrientation(
    Orientation Vector=/Niagara/DynamicInputs/Random/RandomVector(
     Vector Scale:NiagaraFloat=1
     Constraint Vector:Vector3f=<0 0 1>
     Randomness Mode:ENiagaraRandomnessMode=::NewEnumerator0
     Random Seed:NiagaraInt32=0
     Fixed Random Seed:NiagaraBool=false
    )
    Use Orientation Vector=true
    Use Rotation Vector=false
   ) {module=InitialMeshOrientation001}
   &SystemLocation(MaskSpawnGroup=false, Offset=<0 -5 0>)
   &AddVelocity(Velocity=&RotateVectorByQuaternion(
     Vector To Rotate=<150 0 0>
     Quaternion=&FindQuatBetweenTwoVectors(From Vector=<-1 0 0>, To Vector=$Engine.Owner.Velocity)
    )) {enabled=false}
   &InheritVelocity(Module.Limit Speed=true, Velocity Limit=1000, Velocity Scale=<0.5 0.5 0>)
   &DynamicMaterialParameters(
    Index 0 Param 1=1
    Index 0 Param 2=1
    Index 0 Param 3=32
    Index 0 Param 4=8
    Module.Param0WriteEnabled=true
    Module.Param1WriteEnabled=true
    Module.Param2WriteEnabled=true
    Module.Param3WriteEnabled=true
   )
  ]
  particleUpdateScript [
   &ParticleState()
   &GravityForce(Gravity=&VectorFromCurve(VectorCurve=DataInterface() {propertyValues {XCurve {keys [<at 0 0>,<at 1 0>]}, YCurve {keys [<at 0 0>,<at 1 0>]}, ZCurve {keys [<at 0 50 RCIM_Cubic>,<at 0.5 100>]}}}))
   &SolveForcesAndVelocity(Clamp Velocity=false, Limit Acceleration=false)
   &ScaleColor(
    Scale Alpha=&FloatFromCurve(FloatCurve=DataInterface() {propertyValues {Curve {keys [
         <at 0 -1>
         <at 0.15 0>
         {arriveTangent=0.0425, interpMode=RCIM_Cubic, leaveTangent=0.0425, tangentMode=RCTM_User, time=0.3, value=0}
         <at 1 -1>
        ]}}})
    ScaleA=true
    ScaleRGB=true
   )
   &ScaleMeshSize(
    Initial Mesh Scale:Vector3f=$Particles.Scale
    Scale Factor=&VectorFromFloat(Value=&FloatFromCurve(FloatCurve=DataInterface() {propertyValues {Curve {keys [
          {interpMode=RCIM_Cubic, tangentMode=RCTM_User, time=-0.0002, value=0.0422}
          {arriveTangent=0.8781, interpMode=RCIM_Cubic, leaveTangent=0.8781, tangentMode=RCTM_User, time=0.1887, value=0.4675}
          <at 0.9848 0.6888>
         ]}}}))
   ) {enabled=false}
   &ScaleMeshSize(
    Initial Mesh Scale:Vector3f=$Particles.Scale
    Scale Factor=&VectorFromFloat(Value=&FloatFromCurve(FloatCurve=DataInterface() {propertyValues {Curve {keys [{interpMode=RCIM_Cubic, tangentMode=RCTM_User, time=-0.0002, value=0.0422}, <at 0.9848 0.6>]}}}))
   ) {module=ScaleMeshSize001}
  ]
  rendererClasses [&NiagaraMeshRendererProperties]
  renderers [{rendererClass=&NiagaraMeshRendererProperties, rendererIndex=0}]
 }
 DirectionalSpawning_Test {
  bEnabled=false
  emitterUpdateScript [
   &EmitterState()
   &self:SetEngineOwnerVel() {enabled=false, module=SetEngineOwnerVel}
   /SetVariables(Emitter.DirectionTest:NiagaraInt32=&self:DirectionTest(DotLimit:NiagaraFloat=0.9)) {enabled=false, isSetParameters=true, module=SetVariables_#1}
   &self:DT_DirectionTest() {enabled=false, module=DT_DirectionTest}
   &SpawnRate(
    Spawn Probability=&Multiply_Float(A=1, B=&MakeCustomFloatFromBool(Bool=$User.bIsInWater, True Float=1, False Float=0))
    SpawnRate=100
    Use Spawn Probability=true
   ) {enabled=false, module=SpawnRate001}
   &self:DirectionalSpawning(ActivateBool:NiagaraBool=$User.bIsInWater, Dot Tolerence 0-1=0.5) {module=DirectionalSpawning}
  ]
  particleSpawnScript [
   &InitializeParticle(
    Color=&MultiplyLinearColor(Linear Color=<rgba 1 1 1 1>, Scale Factor=<1 1 1 1>)
    Lifetime=&Multiply_Float(A=&UniformRangedFloat(Minimum=0.8, Maximum=1), B=$User.EffectScale)
    Mass=&UniformRangedFloat(Minimum=0.6, Maximum=1.2)
    Mesh Scale=&Multiply_VectorByFloat(Vector=&UniformRangedVector(Minimum=<0.15 0.15 0.1>, Maximum=<0.15 0.2 0.1>), Float=1)
    Module.Write Color=true
    Module.Write Lifetime=true
    Module.Write Mass=true
    Module.Write Material Random=true
    Module.Write Position=false
    Module.Write Scale=true
    Module.Write SpriteRotation=true
    Module.Write SpriteSize=false
    Position=&SimulationPosition()
    Sprite Rotation=&NormalizedAngleToDegrees(NormalizedAngle=&UniformRangedFloat(Minimum=0, Maximum=1))
   )
   /SetVariables(Particles.MeshScale:Vector3f=<4 4 4>) {isSetParameters=true, module=SetVariables_#2}
   &TorusLocation(Align Torus With Axis=false, Large Radius=25, MaskSpawnGroup=false, Module.Override Randomness=false, Torus Mode=::NewEnumerator2) {enabled=false}
   &ApplyInitialForces()
   &InitialMeshOrientation(
    Orientation Axis=<0 0 0>
    Orientation Vector=$Particles.Velocity
    Rotation=<0.1 0 -0.15>
    Use Orientation Vector=true
    Use Rotation Vector=true
   ) {enabled=false, module=InitialMeshOrientation001}
   /SetVariables(Particles.MeshOrientation=&FindQuatBetweenTwoVectors(From Vector=<1 0 0>, To Vector=$Engine.Owner.Velocity)) {isSetParameters=true, module=SetVariables_#3}
   &SystemLocation(MaskSpawnGroup=false, Offset Coordinate Space=::Simulation)
  ]
  particleUpdateScript [
   &ParticleState()
   &DynamicMaterialParameters(
    Index 0 Param 1=&FloatFromCurve(FloatCurve=DataInterface() {propertyValues {Curve {keys [
         {arriveTangent=0.8954, interpMode=RCIM_Cubic, leaveTangent=0.8954, tangentMode=RCTM_User, time=-0.0005, value=-1.845}
         {arriveTangent=1.8419, interpMode=RCIM_Cubic, leaveTangent=1.8419, time=0.7887, value=-0.7711}
         {arriveTangent=3.1029, interpMode=RCIM_Cubic, leaveTangent=3.1029, tangentMode=RCTM_User, time=1.0011, value=0}
        ]}}})
    Index 0 Param 2=&FloatFromCurve(FloatCurve=DataInterface() {propertyValues {Curve {keys [<at -0.0046 4>,<at 1.0011 0>]}}})
    Index 0 Param 3=&FloatFromCurve(
     FloatCurve=DataInterface() {propertyValues {Curve {keys [
         {arriveTangent=0.096, interpMode=RCIM_Cubic, leaveTangent=0.096, tangentMode=RCTM_User, time=-0.0046, value=0.0882}
         <at 1.0011 0.1372 RCIM_Cubic>
        ]}}}
     Scale Curve=&Multiply_Float(A=2, B=$User.EffectScale)
    )
    Index 0 Param 4=&FloatFromCurve(FloatCurve=DataInterface() {propertyValues {Curve {keys [
         {arriveTangent=1.9125, interpMode=RCIM_Cubic, leaveTangent=0.2728, tangentMode=RCTM_User, time=-0.0046, value=0.9276}
         {arriveTangent=4.6326, interpMode=RCIM_Cubic, leaveTangent=4.6326, tangentMode=RCTM_User, time=1.0011, value=2.7648}
        ]}}})
    Module.Param0WriteEnabled=true
    Module.Param1WriteEnabled=true
    Module.Param2WriteEnabled=true
    Module.Param3WriteEnabled=true
   ) {module=DynamicMaterialParameters001}
   &ScaleColor(
    Scale Alpha=&FloatFromCurve(FloatCurve=DataInterface() {propertyValues {Curve {keys [<at -0.0046 1.0113 RCIM_Cubic>,<at 1.0011 0.002 RCIM_Cubic>]}}})
    ScaleA=true
    ScaleRGB=false
   )
   &Drag(Drag=2, Use Linear Drag=true, Use Rotational Drag=false)
   &SolveForcesAndVelocity(Clamp Velocity=false, Limit Acceleration=false)
   &ScaleMeshSize(Scale Factor=&VectorFromCurve(VectorCurve=DataInterface() {propertyValues {
       XCurve {keys [<at 0.0044 1.4881>,<at 0.9989 3>]}
       YCurve {keys [<at -0.0012 1.4698>,<at 0.9989 3>]}
       ZCurve {keys [<at 0.0044 1.5612>,<at 0.9989 1.5622>]}
      }})) {module=MeshSizeScale}
  ]
  rendererClasses [&NiagaraMeshRendererProperties]
  renderers [{rendererClass=&NiagaraMeshRendererProperties, rendererIndex=0}]
 }
 DirectionalSpawning_Test2 {
  bEnabled=false
  emitterUpdateScript [
   &EmitterState(Life Cycle Mode=::NewEnumerator1, Loop Behavior=::NewEnumerator1, Loop Duration Mode=::NewEnumerator1, UseLoopDelay=false)
   &SpawnBurst_Instantaneous(Spawn Count=1, Use Spawn Probability=false)
  ]
  particleSpawnScript [/Niagara/Modules/Spawn/Initialization/V2/InitializeParticle(
    Color Mode=::NewEnumerator0
    Lifetime=5
    Mesh Scale Mode=::NewEnumerator3
    Mesh Uniform Scale=2
    UsePositionOffset=false
    Write Material Random=true
   )]
  particleUpdateScript [
   &ParticleState(Kill Particles When Lifetime Has Elapsed=false)
   &SystemLocation(MaskSpawnGroup=false)
   /SetVariables(Particles.MeshOrientation=&FindQuatBetweenTwoVectors(From Vector=<1 0 0>, To Vector=$Engine.Owner.Velocity)) {isSetParameters=true}
   &self:DirectionalSpawning_Particles(Dot Tolerence 0-1=0.5) {module=DirectionalSpawning_Particles}
  ]
  rendererClasses [&NiagaraSpriteRendererProperties, &NiagaraMeshRendererProperties]
  renderers [
   {rendererClass=&NiagaraSpriteRendererProperties}
   {rendererClass=&NiagaraMeshRendererProperties, rendererIndex=1}
  ]
 }
 FN_Water_Drops {
  bEnabled=false
  emitterUpdateScript [
   &EmitterState()
   &SpawnRate(
    Spawn Probability=&MakeCustomFloatFromBool(Bool=$User.bIsInWater, True Float=1, False Float=0)
    SpawnRate=10
    Use Spawn Probability=true
   )
   &SpawnPerUnit(
    Module.Use Max Movement Threshold=true
    Module.Use Movement Tolerance=true
    Movement Tolerance=1
    Spawn Per Unit:NiagaraFloat=&Multiply_Float(A=&MakeCustomFloatFromBool(Bool=$User.bIsInWater, True Float=1, False Float=0), B=10)
    Use Spawn Probability=false
   )
  ]
  particleSpawnScript [
   /SetVariables(Particles.Random=&UniformRangedFloat(Minimum=0, Maximum=1)) {isSetParameters=true, module=SetVariables_#4}
   /Niagara/Modules/Spawn/Initialization/InitializeParticle(
    Color=<rgba 5 5 5 1>
    Lifetime=&UniformRangedFloat(Minimum=0.5, Maximum=0.75)
    Mass=&UniformRangedFloat(Minimum=0.6, Maximum=1.2)
    Mesh Scale=<100 100 100>
    Module.Write Color=true
    Module.Write Lifetime=true
    Module.Write Mass=true
    Module.Write Material Random=true
    Module.Write Position=false
    Module.Write Scale=false
    Module.Write SpriteRotation=true
    Module.Write SpriteSize=true
    Position=&SimulationPosition()
    Sprite Size=&Vector2DFromCurve(
     Vector2Curve=DataInterface() {propertyValues {
       XCurve {keys [<at 0 1 RCIM_Cubic>,<at 1 0.3 RCIM_Cubic>]}
       YCurve {keys [{arriveTangent=-4.9074, interpMode=RCIM_Cubic, leaveTangent=-4.9074, tangentMode=RCTM_User, time=0, value=3}, <at 1 2 RCIM_Cubic>]}
      }}
     CurveIndex=$Particles.Random
     ScaleCurve=<0.25 0.25>
    )
   )
   &DynamicMaterialParameters(
    Index 0 Param 1=&UniformRangedFloat(Minimum=4, Maximum=12)
    Module.Param0WriteEnabled=true
    Module.Param1WriteEnabled=false
    Module.Param2WriteEnabled=false
    Module.Param3WriteEnabled=false
   )
   &InitializeWaterSnapToOwner()
   &ConstrainPositionToPlane(Plane Normal:Vector3f=$Particles.WaterPlaneNormal, Plane Position:NiagaraPosition=$User.WaterPlaneLocation) {module=ConstrainPositionToPlane001}
   &CylinderLocation(
    Cylinder Height=1
    Cylinder Radius=40
    Lathe Profile=DataInterface() {propertyValues {Curve {keys []}}}
    MaskSpawnGroup=false
    Module.Enable Lathe Profile=false
    Module.Override Randomness=false
    Module.Surface Only=true
    Surface Only Band Thickness=2
   )
   /SetVariables(Particles.SpriteSize=&Multiply_Vector2(
     A=$Particles.SpriteSize
     B=&Vector2DFromFloat(Value=&FloatFromCurve(
       FloatCurve=DataInterface() {propertyValues {Curve {keys [
           {arriveTangent=1.6964, interpMode=RCIM_Cubic, time=-0.0012, value=-0.0071}
           {arriveTangent=13.7683, interpMode=RCIM_Cubic, leaveTangent=13.7683, time=0.3287, value=17.2263}
           {arriveTangent=-3.3438, interpMode=RCIM_Cubic, leaveTangent=-3.3438, time=0.7139, value=9.8392}
           {arriveTangent=1.8535, interpMode=RCIM_Cubic, leaveTangent=1.8535, tangentMode=RCTM_User, time=0.9945, value=15}
          ]}}}
       CurveIndex=$Emitter.Scalability.DistanceFraction
      ))
    )) {isSetParameters=true}
   &PointForce(
    Force Origin=&SimulationPosition()
    Force Strength=&UniformRangedFloat(Minimum=200, Maximum=300)
    Module.Use Falloff=false
    Module.UseLinearForce=false
   ) {module=PointForce001}
   &ApplyInitialForces()
  ]
  particleUpdateScript [
   &ParticleState()
   /Game/Effects/Niagara/Emitters/Water/FN_DistanceKillParticles(AgeReduiceAmount=0.01)
   &ScaleColor(
    Scale Alpha=&FloatFromCurve(FloatCurve=DataInterface() {propertyValues {Curve {keys [
         <at 0 0 RCIM_Cubic>
         {arriveTangent=0.002, interpMode=RCIM_Cubic, leaveTangent=0.002, time=0.1183, value=1.0113}
         <at 1.0011 0.002 RCIM_Cubic>
        ]}}})
    ScaleA=true
    ScaleRGB=true
   )
   /Niagara/Modules/Update/Forces/CurlNoiseForce(
    Noise Frequency=1
    Noise Strength=&FloatFromCurve(FloatCurve=DataInterface() {propertyValues {Curve {keys [
         {arriveTangent=-1159.6794, interpMode=RCIM_Cubic, leaveTangent=-1159.6792, tangentMode=RCTM_User, time=-0.0046, value=500}
         {arriveTangent=-447.4286, interpMode=RCIM_Cubic, leaveTangent=-447.4286, time=0.2223, value=189.6793}
         <at 1.0011 50 RCIM_Cubic>
        ]}}})
    Pan Noise=false
   )
   &Drag(Use Linear Drag=true, Use Rotational Drag=false)
   &GravityForce()
   &SolveForcesAndVelocity(Clamp Velocity=false, Limit Acceleration=false)
  ]
  rendererClasses [&NiagaraSpriteRendererProperties]
  renderers [{rendererClass=&NiagaraSpriteRendererProperties}]
 }
 FN_Water_Drops001 {
  bEnabled=false
  emitterUpdateScript [
   &EmitterState()
   &EmitterScalability(
    MaxDistance=800
    Module.bApplySpawnCountScale=true
    Module.bApplySpawnCountScaleByDistanceFraction=true
    Module.bUseMaxDistance=true
    Module.bUseMinDistance=false
    SpawnCountScaleByDistanceFraction=DataInterface() {propertyValues {Curve {keys [<at 0 1>,<at 1 0.3>]}}}
   ) {enabled=false}
   &SpawnBurst_Instantaneous(
    Spawn Count=&RandomRangeInt(Minimum=5, Maximum=10)
    Spawn Probability=&Multiply_Float(A=0.75, B=&MakeCustomFloatFromBool(Bool=$User.bIsInWater, True Float=1, False Float=0))
    Use Spawn Probability=true
   )
   &EmitterLifeCycle(MaxLoopCount=1, NextLoopDuration=0.25) {enabled=false}
   &SpawnRate(SpawnRate=10, Use Spawn Probability=false) {enabled=false}
  ]
  particleSpawnScript [
   /SetVariables(Particles.Random=&UniformRangedFloat(Minimum=0, Maximum=1)) {isSetParameters=true, module=SetVariables_#4}
   &InitializeParticle(
    Color=<rgba 1 1 1 0.4>
    Lifetime=&UniformRangedFloat(Minimum=1, Maximum=2)
    Mass=&UniformRangedFloat(Minimum=0.6, Maximum=1.2)
    Mesh Scale=<100 100 100>
    Module.Write Color=true
    Module.Write Lifetime=true
    Module.Write Mass=true
    Module.Write Material Random=true
    Module.Write Position=false
    Module.Write Scale=false
    Module.Write SpriteRotation=true
    Module.Write SpriteSize=true
    Position=&SimulationPosition()
    Sprite Size=&Vector2DFromCurve(
     Vector2Curve=DataInterface() {propertyValues {
       XCurve {keys [<at -0.0011 2.0768 RCIM_Cubic>,<at 0.9989 0.3489 RCIM_Cubic>]}
       YCurve {keys [
         {arriveTangent=-30.8979, interpMode=RCIM_Cubic, leaveTangent=-30.8979, tangentMode=RCTM_User, time=0, value=15.3962}
         {arriveTangent=-13.4788, interpMode=RCIM_Cubic, leaveTangent=-13.4788, time=0.2382, value=6.3246}
         <at 0.9989 1.9328 RCIM_Cubic>
        ]}
      }}
     CurveIndex=$Particles.Random
     ScaleCurve=<1 1>
    )
   )
   /Niagara/Modules/Spawn/Orientation/V2/InitialMeshOrientation(
    Orientation Vector=<0 0 0>
    Rotation=&UniformRangedVector(Minimum=<0 0 0>, Maximum=<0 0 1>)
    Use Orientation Vector=true
    Use Rotation Vector=true
   ) {enabled=false}
   &DynamicMaterialParameters(
    Index 0 Param 1=&UniformRangedFloat(Minimum=4, Maximum=10)
    Module.Param0WriteEnabled=true
    Module.Param1WriteEnabled=false
    Module.Param2WriteEnabled=false
    Module.Param3WriteEnabled=false
   )
   &InheritVelocity(Module.Limit Speed=false, Velocity Limit=800, Velocity Scale=<1 1 -0.3>) {enabled=false}
   &AddVelocity(
    Coordinate Space=::World
    Velocity=&RotateVectorByQuaternion(
     Vector To Rotate=&UniformRangedVector(Minimum=<150 -150 300>, Maximum=<150 150 300>)
     Quaternion=&FindQuatBetweenTwoVectors(From Vector=<1 0 0>, To Vector=$Engine.Owner.Velocity)
    )
   )
   &CylinderLocation(
    Cylinder Height=1
    Cylinder Radius=20
    Lathe Profile=DataInterface() {propertyValues {Curve {keys []}}}
    MaskSpawnGroup=false
    Module.Enable Lathe Profile=false
    Module.Override Randomness=false
    Module.Surface Only=true
    Surface Only Band Thickness=2
   ) {enabled=false}
   &InitializeWaterSnapToOwner()
   &ConstrainPositionToPlane(Plane Normal:Vector3f=$Particles.WaterPlaneNormal, Plane Position:NiagaraPosition=$User.WaterPlaneLocation) {module=ConstrainPositionToPlane001}
  ]
  particleUpdateScript [
   &ParticleState()
   &FN_DistanceKillParticles(AgeReduiceAmount=0.01) {enabled=false}
   &ScaleColor(
    Scale Alpha=&FloatFromCurve(FloatCurve=DataInterface() {propertyValues {Curve {keys [
         <at 0 0 RCIM_Cubic>
         {arriveTangent=0.002, interpMode=RCIM_Cubic, leaveTangent=0.002, time=0.1183, value=1.0113}
         <at 1.0011 0.002 RCIM_Cubic>
        ]}}})
    ScaleA=true
    ScaleRGB=true
   )
   &PointForce(Force Origin=&SimulationPosition(), Force Strength=200, Module.Use Falloff=false, Module.UseLinearForce=false) {enabled=false}
   &CurlNoiseForce(
    Noise Frequency=1
    Noise Strength=&FloatFromCurve(FloatCurve=DataInterface() {propertyValues {Curve {keys [
         {arriveTangent=-1159.6794, interpMode=RCIM_Cubic, leaveTangent=-1159.6792, tangentMode=RCTM_User, time=-0.0046, value=500}
         {arriveTangent=-447.4286, interpMode=RCIM_Cubic, leaveTangent=-447.4286, time=0.2223, value=189.6793}
         <at 1.0011 50 RCIM_Cubic>
        ]}}})
    Pan Noise=false
   ) {enabled=false}
   &Drag(Use Linear Drag=true, Use Rotational Drag=false)
   &GravityForce(Gravity=<0 0 -400>)
   &SolveForcesAndVelocity(Clamp Velocity=false, Limit Acceleration=false)
   &ScaleMeshSize() {enabled=false, module=MeshSizeScale}
   &ConstrainPositionToPlane(Plane Normal:Vector3f=$Particles.WaterPlaneNormal, Plane Position:Vector3f=$Particles.WaterPlaneLocation) {enabled=false}
  ]
  rendererClasses [&NiagaraSpriteRendererProperties, &NiagaraMeshRendererProperties]
  renderers [
   {rendererClass=&NiagaraSpriteRendererProperties}
   {rendererClass=&NiagaraMeshRendererProperties, rendererIndex=1}
  ]
 }
 FN_Water_Drops002 {
  emitterUpdateScript [
   &EmitterState()
   &self:DirectionalSpawning(
    ActivateBool:NiagaraBool=$User.bIsInWater
    Dot Tolerence 0-1=0.5
    SpawnCount:NiagaraInt32=&RandomRangeInt(Minimum=3, Maximum=5, Evaluation Type=::NewEnumerator1)
   ) {module=DirectionalSpawning}
  ]
  particleSpawnScript [
   /SetVariables(Particles.Random=&UniformRangedFloat(Minimum=0, Maximum=1)) {isSetParameters=true, module=SetVariables_#4}
   &InitializeParticle(
    Color=<rgba 1 1 1 0.4>
    Lifetime=&UniformRangedFloat(Minimum=0.4, Maximum=0.6)
    Mass=&UniformRangedFloat(Minimum=0.6, Maximum=1.2)
    Mesh Scale=<100 100 100>
    Module.Write Color=true
    Module.Write Lifetime=true
    Module.Write Mass=false
    Module.Write Material Random=true
    Module.Write Position=false
    Module.Write Scale=false
    Module.Write SpriteRotation=true
    Module.Write SpriteSize=true
    Position=&SimulationPosition()
    Sprite Size=&MakeVector2D(
     X=&RandomRangeFloat(Minimum=1.5, Maximum=2.5)
     Y=&RandomRangeFloat(Minimum=4, Maximum=6)
    )
   )
   &DynamicMaterialParameters(
    Index 0 Param 1=&UniformRangedFloat(Minimum=4, Maximum=10)
    Module.Param0WriteEnabled=true
    Module.Param1WriteEnabled=false
    Module.Param2WriteEnabled=false
    Module.Param3WriteEnabled=false
   )
   &AddVelocity(
    Coordinate Space=::World
    Velocity=&RotateVectorByQuaternion(
     Vector To Rotate=&UniformRangedVector(Minimum=<75 -150 175>, Maximum=<125 150 250>)
     Quaternion=&FindQuatBetweenTwoVectors(From Vector=<1 0 0>, To Vector=$Engine.Owner.Velocity)
    )
   )
   &InitializeWaterSnapToOwner()
   &ConstrainPositionToPlane(Plane Normal:Vector3f=$Particles.WaterPlaneNormal, Plane Position:NiagaraPosition=$User.WaterPlaneLocation) {module=ConstrainPositionToPlane001}
  ]
  particleUpdateScript [
   &ParticleState()
   &ScaleColor(
    Scale Alpha=&FloatFromCurve(FloatCurve=DataInterface() {propertyValues {Curve {keys [
         <at 0 0 RCIM_Cubic>
         {arriveTangent=0.002, interpMode=RCIM_Cubic, leaveTangent=0.002, time=0.1183, value=1.0113}
         <at 1.0011 0.002 RCIM_Cubic>
        ]}}})
    ScaleA=true
    ScaleRGB=true
   )
   &Drag(Use Linear Drag=true, Use Rotational Drag=false)
   &GravityForce(Gravity=<0 0 -400>)
   &SolveForcesAndVelocity(Clamp Velocity=false, Limit Acceleration=false)
  ]
  rendererClasses [&NiagaraSpriteRendererProperties]
  renderers [{rendererClass=&NiagaraSpriteRendererProperties}]
 }
 FN_Water_Foam_SnapToSurface {
  emitterUpdateScript [
   &EmitterState()
   &SpawnRate(
    Spawn Probability=&MakeCustomFloatFromBool(Bool=$User.bIsInWater, True Float=1, False Float=0)
    SpawnRate=4
    Use Spawn Probability=true
   )
  ]
  particleSpawnScript [
   &InitializeParticle(
    Mesh Scale=&VectorFromFloat(Value=&Multiply_Float(A=25, B=$User.EffectScale))
    Module.Write Color=true
    Module.Write Lifetime=true
    Module.Write Mass=false
    Module.Write Material Random=true
    Module.Write Position=false
    Module.Write Scale=true
    Module.Write SpriteRotation=true
    Module.Write SpriteSize=false
    Position=&SimulationPosition()
    Sprite Rotation=&NormalizedAngleToDegrees(NormalizedAngle=&UniformRangedFloat(Minimum=0, Maximum=1))
   )
   /Niagara/Modules/Spawn/Orientation/V3/InitialMeshOrientation(
    Orientation Vector=<0 0 0>
    Rotation=&UniformRangedVector(Minimum=<0 0 0>, Maximum=<0 0 1>)
    Use Orientation Vector=true
    Use Rotation Vector=true
   ) {module=InitialMeshOrientation001}
   &DynamicMaterialParameters(
    Index 1 Param 3=200
    Index 3 Param 1=$User.WaterBodyIndex
    Index 3 Param 2=$User.WaterDepth
    Module.Param0WriteEnabled1=false
    Module.Param0WriteEnabled3=true
    Module.Param1WriteEnabled1=false
    Module.Param1WriteEnabled3=true
    Module.Param2WriteEnabled1=true
    Module.Param2WriteEnabled3=true
    Module.Param3WriteEnabled1=false
    Module.Param3WriteEnabled3=true
    Write Parameter Index 0=false
    Write Parameter Index 1=true
    Write Parameter Index 3=true
   )
   &InitializeWaterSnapToOwner()
   &InheritVelocity(Module.Limit Speed=true)
  ]
  particleUpdateScript [
   &ParticleState()
   &DynamicMaterialParameters(
    Index 0 Param 1=&FloatFromCurve(FloatCurve=DataInterface() {propertyValues {Curve {keys [<at -0.0046 1>,<at 1.0011 4>]}}})
    Index 0 Param 2=&FloatFromCurve(FloatCurve=DataInterface() {propertyValues {Curve {keys [{arriveTangent=0.6218, interpMode=RCIM_Cubic, leaveTangent=0.6218, tangentMode=RCTM_User, time=-0.0046, value=0.5}, <at 1.0011 1 RCIM_Cubic>]}}})
    Index 0 Param 3=&FloatFromCurve(FloatCurve=DataInterface() {propertyValues {Curve {keys [<at -0.0046 1>,<at 1.0011 0.5>]}}})
    Index 2 Param 1=&FloatFromCurve(FloatCurve=DataInterface() {propertyValues {Curve {keys [<at -0.0046 0.4>,<at 1.0011 1.5>]}}})
    Index 2 Param 2=&FloatFromCurve(FloatCurve=DataInterface() {propertyValues {Curve {keys [<at -0.0046 0.1>,<at 0.11 1.75>,<at 1.0011 0.3>]}}})
    Index 2 Param 3=1
    Index 2 Param 4=&FloatFromCurve(FloatCurve=DataInterface() {propertyValues {Curve {keys [
         {arriveTangent=-1.9991, interpMode=RCIM_Cubic, time=0, value=1}
         {arriveTangent=-2.7143, interpMode=RCIM_Cubic, leaveTangent=-2.7143, tangentMode=RCTM_User, time=0.2109, value=0.5898}
         <at 1 0 RCIM_Cubic>
        ]}}})
    Module.Param0WriteEnabled=true
    Module.Param0WriteEnabled2=true
    Module.Param1WriteEnabled=true
    Module.Param1WriteEnabled2=true
    Module.Param2WriteEnabled=true
    Module.Param2WriteEnabled2=true
    Module.Param3WriteEnabled=true
    Module.Param3WriteEnabled2=true
    Write Parameter Index 2=true
   ) {module=DynamicMaterialParameters001}
   &Drag(Drag=2, Use Linear Drag=true, Use Rotational Drag=false)
   &SolveForcesAndVelocity(Clamp Velocity=false, Limit Acceleration=false)
   &ScaleMeshSize(Scale Factor=&VectorFromFloat(Value=&FloatFromCurve(FloatCurve=DataInterface() {propertyValues {Curve {keys [
          {arriveTangent=26.4511, interpMode=RCIM_Cubic, leaveTangent=26.451, tangentMode=RCTM_User, time=-0.0046, value=0.6073}
          {arriveTangent=3.3733, interpMode=RCIM_Cubic, leaveTangent=3.3733, time=0.1268, value=2.5016}
          <at 1.0011 4 RCIM_Cubic>
         ]}}}))) {module=MeshSizeScale}
   &ConstrainPositionToPlane(
    Plane Normal:Vector3f=$Particles.WaterPlaneNormal
    Plane Position:NiagaraPosition=&Add_Vector(A=Linked() {linkedVariable {name=User.WaterPlaneLocation, type=NiagaraPosition}}, B=<0 0 0>)
   )
   &CameraOffset(Camera Offset Amount=10)
  ]
  rendererClasses [&NiagaraMeshRendererProperties]
  renderers [{rendererClass=&NiagaraMeshRendererProperties, rendererIndex=0}]
 }
 Procedural_Dust_by_Direction {
  bEnabled=false
  emitterUpdateScript [
   &EmitterState()
   &SpawnBurst_Instantaneous(Spawn Count=10, Spawn Time=0.05, Use Spawn Probability=false) {enabled=false, module=SpawnBurst_Instantaneous001}
   &self:DirectionalSpawning(
    ActivateBool:NiagaraBool=&InvertBool(Boolean:NiagaraBool=$User.bIsInWater)
    Dot Tolerence 0-1=0.5
    SpawnCount:NiagaraInt32=&RandomRangeInt(Minimum=3, Maximum=5, Evaluation Type=::NewEnumerator1)
   ) {module=DirectionalSpawning}
  ]
  particleSpawnScript [
   &InitializeParticle(
    Color=&UniformRangedLinearColor(Minimum=<rgba 0.6094 0.5279 0.4433 0.1>, Maximum=<rgba 0.6094 0.5279 0.4433 0.05>)
    Lifetime=&UniformRangedFloat(Minimum=0.25, Maximum=0.5)
    Module.Write Color=true
    Module.Write Lifetime=true
    Module.Write Mass=true
    Module.Write Material Random=true
    Module.Write Position=false
    Module.Write Scale=false
    Module.Write SpriteRotation=true
    Module.Write SpriteSize=true
    Position=&SimulationPosition()
    Sprite Rotation=&UniformRangedFloat(Minimum=0, Maximum=360)
    Sprite Size=&Vector2DFromFloat(Value=&UniformRangedFloat(Minimum=5, Maximum=15))
   )
   &AddVelocity(
    Coordinate Space=::World
    Velocity=&RotateVectorByQuaternion(
     Vector To Rotate=&UniformRangedVector(Minimum=<50 -50 5>, Maximum=<100 50 25>)
     Quaternion=&FindQuatBetweenTwoVectors(From Vector=<-1 0 0>, To Vector=$Engine.Owner.Velocity)
    )
   ) {module=AddVelocity001}
   &SphereLocation(
    MaskSpawnGroup=false
    Override Seed=false
    Sphere Radius=10
    Surface Only=false
    UseNonUniformScale=false
    UseOffset=false
    UseSphereOrientation=false
   )
  ]
  particleUpdateScript [
   &ParticleState()
   &Drag(Drag=3, Use Linear Drag=true, Use Rotational Drag=false) {enabled=false}
   &AccelerationForce(Acceleration=&RandomRangeVector(Minimum=<0 0 20>, Maximum=<0 0 50>)) {module=AccelerationForce001}
   &SolveForcesAndVelocity(Clamp Velocity=false, Limit Acceleration=false)
   &ScaleSpriteSize(Scale Factor=&Vector2DFromFloat(Value=&FloatFromCurve(FloatCurve=DataInterface() {propertyValues {Curve {keys [
          {arriveTangent=1.491, interpMode=RCIM_Cubic, leaveTangent=1.491, tangentMode=RCTM_User, time=-0.0046, value=1.1969}
          <at 1.0011 3.4809 RCIM_Cubic>
         ]}}})))
   /SetVariables(Particles.SubImageIndex=&FloatFromCurve(FloatCurve=DataInterface() {propertyValues {Curve {keys [
         {arriveTangent=29.8623, interpMode=RCIM_Cubic, leaveTangent=29.8623, tangentMode=RCTM_User, time=-0.0046, value=5.0556}
         <at 1.0011 22.1689 RCIM_Cubic>
        ]}}})) {isSetParameters=true, module=SetVariables_#5}
   &DynamicMaterialParameters(
    Index 0 Param 1=&FloatFromCurve(FloatCurve=DataInterface() {propertyValues {Curve {keys [<at -0.0046 1.0113 RCIM_Cubic>,<at 1.0011 0.002>]}}})
    Index 0 Param 4=&FloatFromCurve(FloatCurve=DataInterface() {propertyValues {Curve {keys [<at 0 0 RCIM_Cubic>,<at 1 0.5>]}}})
    Module.Param0WriteEnabled=true
    Module.Param1WriteEnabled=false
    Module.Param2WriteEnabled=false
    Module.Param3WriteEnabled=true
   ) {module=DynamicMatParams}
   &CameraOffset(Camera Offset Amount=10)
   &ScaleColor(
    Scale Alpha=&FloatFromCurve(FloatCurve=DataInterface() {propertyValues {Curve {keys [
         {arriveTangent=-3.1757, interpMode=RCIM_Cubic, leaveTangent=-3.1757, tangentMode=RCTM_User, time=0.5, value=1.0113}
         {arriveTangent=-1.5121, interpMode=RCIM_Cubic, leaveTangent=-1.5121, tangentMode=RCTM_User, time=1.0011, value=0.002}
        ]}}})
    ScaleA=true
    ScaleRGB=false
   )
  ]
  rendererClasses [&NiagaraSpriteRendererProperties]
  renderers [{rendererClass=&NiagaraSpriteRendererProperties}]
 }
 Smoke001 {
  bEnabled=false
  emitterUpdateScript [
   &EmitterState()
   &SpawnBurst_Instantaneous(Spawn Count=10, Spawn Time=0.05, Use Spawn Probability=false) {enabled=false, module=SpawnBurst_Instantaneous001}
   &self:DirectionalSpawning(
    ActivateBool:NiagaraBool=&InvertBool(Boolean:NiagaraBool=$User.bIsInWater)
    Dot Tolerence 0-1=0.5
    SpawnCount:NiagaraInt32=&RandomRangeInt(Minimum=3, Maximum=5, Evaluation Type=::NewEnumerator1)
   ) {module=DirectionalSpawning}
  ]
  particleSpawnScript [
   &InitializeParticle(
    Color=&UniformRangedLinearColor(Minimum=<rgba 0.6094 0.5279 0.4433 0.1>, Maximum=<rgba 0.6094 0.5279 0.4433 0.05>)
    Lifetime=&UniformRangedFloat(Minimum=0.25, Maximum=0.5)
    Module.Write Color=true
    Module.Write Lifetime=true
    Module.Write Mass=true
    Module.Write Material Random=true
    Module.Write Position=false
    Module.Write Scale=false
    Module.Write SpriteRotation=true
    Module.Write SpriteSize=true
    Position=&SimulationPosition()
    Sprite Rotation=&UniformRangedFloat(Minimum=0, Maximum=360)
    Sprite Size=&Vector2DFromFloat(Value=&UniformRangedFloat(Minimum=5, Maximum=15))
   )
   &AddVelocity(
    Coordinate Space=::World
    Velocity=&RotateVectorByQuaternion(
     Vector To Rotate=&UniformRangedVector(Minimum=<50 -50 5>, Maximum=<100 50 25>)
     Quaternion=&FindQuatBetweenTwoVectors(From Vector=<-1 0 0>, To Vector=$Engine.Owner.Velocity)
    )
   ) {module=AddVelocity001}
   &SphereLocation(
    MaskSpawnGroup=false
    Override Seed=false
    Sphere Radius=10
    Surface Only=false
    UseNonUniformScale=false
    UseOffset=false
    UseSphereOrientation=false
   )
  ]
  particleUpdateScript [
   &ParticleState()
   &Drag(Drag=3, Use Linear Drag=true, Use Rotational Drag=false) {enabled=false}
   &AccelerationForce(Acceleration=&RandomRangeVector(Minimum=<0 0 20>, Maximum=<0 0 50>)) {module=AccelerationForce001}
   &SolveForcesAndVelocity(Clamp Velocity=false, Limit Acceleration=false)
   &ScaleSpriteSize(Scale Factor=&Vector2DFromFloat(Value=&FloatFromCurve(FloatCurve=DataInterface() {propertyValues {Curve {keys [
          {arriveTangent=1.491, interpMode=RCIM_Cubic, leaveTangent=1.491, tangentMode=RCTM_User, time=-0.0046, value=1.1969}
          <at 1.0011 3.4809 RCIM_Cubic>
         ]}}})))
   /SetVariables(Particles.SubImageIndex=&FloatFromCurve(FloatCurve=DataInterface() {propertyValues {Curve {keys [
         {arriveTangent=29.8623, interpMode=RCIM_Cubic, leaveTangent=29.8623, tangentMode=RCTM_User, time=-0.0046, value=5.0556}
         <at 1.0011 22.1689 RCIM_Cubic>
        ]}}})) {isSetParameters=true, module=SetVariables_#5}
   &DynamicMaterialParameters(
    Index 0 Param 1=&FloatFromCurve(FloatCurve=DataInterface() {propertyValues {Curve {keys [<at -0.0046 1.0113 RCIM_Cubic>,<at 1.0011 0.002>]}}})
    Index 0 Param 4=&FloatFromCurve(FloatCurve=DataInterface() {propertyValues {Curve {keys [<at 0 0 RCIM_Cubic>,<at 1 0.5>]}}})
    Module.Param0WriteEnabled=true
    Module.Param1WriteEnabled=false
    Module.Param2WriteEnabled=false
    Module.Param3WriteEnabled=true
   ) {module=DynamicMatParams}
   &CameraOffset(Camera Offset Amount=10)
   &ScaleColor(
    Scale Alpha=&FloatFromCurve(FloatCurve=DataInterface() {propertyValues {Curve {keys [
         {arriveTangent=-3.1757, interpMode=RCIM_Cubic, leaveTangent=-3.1757, tangentMode=RCTM_User, time=0.5, value=1.0113}
         {arriveTangent=-1.5121, interpMode=RCIM_Cubic, leaveTangent=-1.5121, tangentMode=RCTM_User, time=1.0011, value=0.002}
        ]}}})
    ScaleA=true
    ScaleRGB=false
   )
  ]
  rendererClasses [&NiagaraSpriteRendererProperties]
  renderers [{rendererClass=&NiagaraSpriteRendererProperties}]
 }
 WaterSplashArmL {
  bEnabled=false
  emitterUpdateScript [
   &EmitterState()
   &self:SetEngineOwnerVel() {enabled=false, module=SetEngineOwnerVel}
   /SetVariables(Emitter.DirectionTest:NiagaraInt32=&self:DirectionTest(DotLimit:NiagaraFloat=0.9)) {enabled=false, isSetParameters=true, module=SetVariables_#1}
   &self:DT_DirectionTest() {enabled=false, module=DT_DirectionTest}
   &SpawnRate(
    Spawn Probability=&Multiply_Float(A=1, B=&MakeCustomFloatFromBool(Bool=$User.bIsInWater, True Float=1, False Float=0))
    SpawnRate=100
    Use Spawn Probability=true
   ) {enabled=false, module=SpawnRate001}
   &self:DirectionalSpawning(Dot Tolerence 0-1=1) {module=DirectionalSpawning}
  ]
  particleSpawnScript [
   &InitializeParticle(
    Color=&MultiplyLinearColor(Linear Color=<rgba 1 1 1 1>, Scale Factor=<1 1 1 1>)
    Lifetime=&Multiply_Float(A=&UniformRangedFloat(Minimum=0.5, Maximum=0.7), B=$User.EffectScale)
    Mass=&UniformRangedFloat(Minimum=0.6, Maximum=1.2)
    Mesh Scale=&Multiply_VectorByFloat(Vector=&UniformRangedVector(Minimum=<0.15 0.3 0.1>, Maximum=<0.15 0.3 0.1>), Float=1)
    Module.Write Color=true
    Module.Write Lifetime=true
    Module.Write Mass=true
    Module.Write Material Random=true
    Module.Write Position=false
    Module.Write Scale=true
    Module.Write SpriteRotation=true
    Module.Write SpriteSize=false
    Position=&SimulationPosition()
    Sprite Rotation=&NormalizedAngleToDegrees(NormalizedAngle=&UniformRangedFloat(Minimum=0, Maximum=1))
   )
   /SetVariables(Particles.MeshScale:Vector3f=<4 4 4>) {isSetParameters=true, module=SetVariables_#2}
   &TorusLocation(Align Torus With Axis=false, Large Radius=25, MaskSpawnGroup=false, Module.Override Randomness=false, Torus Mode=::NewEnumerator2)
   &ApplyInitialForces()
   &InitialMeshOrientation(
    Orientation Axis=<0 0 0>
    Orientation Vector=$Particles.Velocity
    Rotation=<0.1 0 -0.15>
    Use Orientation Vector=true
    Use Rotation Vector=true
   ) {module=InitialMeshOrientation001}
   /SetVariables(Particles.MeshOrientation=&FindQuatBetweenTwoVectors(From Vector=<1 0 0>, To Vector=$Engine.Owner.Velocity)) {isSetParameters=true, module=SetVariables_#3}
   &SystemLocation(MaskSpawnGroup=false, Offset Coordinate Space=::Simulation)
  ]
  particleUpdateScript [
   &ParticleState()
   &ParticleState() {module=UpdateAge}
   &DynamicMaterialParameters(
    Index 0 Param 1=&FloatFromCurve(FloatCurve=DataInterface() {propertyValues {Curve {keys [
         {arriveTangent=0.8954, interpMode=RCIM_Cubic, leaveTangent=0.8954, tangentMode=RCTM_User, time=-0.0005, value=-1.845}
         {arriveTangent=1.8419, interpMode=RCIM_Cubic, leaveTangent=1.8419, time=0.7887, value=-0.7711}
         {arriveTangent=3.1029, interpMode=RCIM_Cubic, leaveTangent=3.1029, tangentMode=RCTM_User, time=1.0011, value=0}
        ]}}})
    Index 0 Param 2=&FloatFromCurve(FloatCurve=DataInterface() {propertyValues {Curve {keys [<at -0.0046 4>,<at 1.0011 0>]}}})
    Index 0 Param 3=&FloatFromCurve(
     FloatCurve=DataInterface() {propertyValues {Curve {keys [
         {arriveTangent=0.096, interpMode=RCIM_Cubic, leaveTangent=0.096, tangentMode=RCTM_User, time=-0.0046, value=0.0882}
         <at 1.0011 0.1372 RCIM_Cubic>
        ]}}}
     Scale Curve=&Multiply_Float(A=2, B=$User.EffectScale)
    )
    Index 0 Param 4=&FloatFromCurve(FloatCurve=DataInterface() {propertyValues {Curve {keys [
         {arriveTangent=1.9125, interpMode=RCIM_Cubic, leaveTangent=0.2728, tangentMode=RCTM_User, time=-0.0046, value=0.9276}
         {arriveTangent=4.6326, interpMode=RCIM_Cubic, leaveTangent=4.6326, tangentMode=RCTM_User, time=1.0011, value=2.7648}
        ]}}})
    Module.Param0WriteEnabled=true
    Module.Param1WriteEnabled=true
    Module.Param2WriteEnabled=true
    Module.Param3WriteEnabled=true
   ) {module=DynamicMaterialParameters001}
   &ScaleColor(
    Scale Alpha=&FloatFromCurve(FloatCurve=DataInterface() {propertyValues {Curve {keys [<at -0.0046 1.0113 RCIM_Cubic>,<at 1.0011 0.002 RCIM_Cubic>]}}})
    ScaleA=true
    ScaleRGB=false
   )
   &Drag(Drag=2, Use Linear Drag=true, Use Rotational Drag=false)
   &SolveForcesAndVelocity(Clamp Velocity=false, Limit Acceleration=false)
   &ScaleMeshSize(Scale Factor=&VectorFromCurve(VectorCurve=DataInterface() {propertyValues {
       XCurve {keys [<at 0.0044 1.4881>,<at 0.9989 3>]}
       YCurve {keys [<at -0.0012 1.4698>,<at 0.9989 3>]}
       ZCurve {keys [<at 0.0044 1.5612>,<at 0.9989 1.5622>]}
      }})) {module=MeshSizeScale}
  ]
  rendererClasses [&NiagaraMeshRendererProperties]
  renderers [
   {rendererClass=&NiagaraMeshRendererProperties, rendererIndex=0}
   {rendererClass=&NiagaraMeshRendererProperties, rendererIndex=1}
  ]
 }
}
summary {
 systemName=NS_SoccerTraversal_Emote_WaterSplash
 emitters [
  {bEnabled=false, emitterName=WaterSplashArmL, rendererClasses [&NiagaraMeshRendererProperties]}
  {bEnabled=false, emitterName=FN_Water_Drops, rendererClasses [&NiagaraSpriteRendererProperties]}
  {emitterName=FN_Water_Foam_SnapToSurface, rendererClasses [&NiagaraMeshRendererProperties]}
  {bEnabled=false, emitterName=Bubbles001, rendererClasses [&NiagaraMeshRendererProperties]}
  {
   bEnabled=false
   emitterName=FN_Water_Drops001
   rendererClasses [&NiagaraSpriteRendererProperties, &NiagaraMeshRendererProperties]
  }
  {bEnabled=false, emitterName=DirectionalSpawning_Test, rendererClasses [&NiagaraMeshRendererProperties]}
  {
   bEnabled=false
   emitterName=DirectionalSpawning_Test2
   rendererClasses [&NiagaraSpriteRendererProperties, &NiagaraMeshRendererProperties]
  }
  {emitterName=FN_Water_Drops002, rendererClasses [&NiagaraSpriteRendererProperties]}
  {bEnabled=false, emitterName=Procedural_Dust_by_Direction, rendererClasses [&NiagaraSpriteRendererProperties]}
  {bEnabled=false, emitterName=Smoke001, rendererClasses [&NiagaraSpriteRendererProperties]}
 ]
 userVariables [
  User.bIsInWater {type=NiagaraBool}
  User.EffectScale {type=NiagaraFloat, defaultValue=1, description=""}
  User.WaterBodyIndex {type=NiagaraFloat, defaultValue=10, description=""}
  User.WaterDepth {type=NiagaraFloat, defaultValue=0, description=""}
  User.WaterPlaneLocation {type=NiagaraPosition, description="", defaultValue {}}
  User.WaterPlaneNormal {type=Vector3f, description="", defaultValue {struct=Vector3f, value=<0 0 0>}}
  User.WaveAttenuationFactor {type=NiagaraFloat, defaultValue=0, description=""}
  User.WaveTime {type=NiagaraFloat, defaultValue=0, description=""}
 ]
}
